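# Niagara ambient effect of dove flocks flying by in the Olympus environment, using one GPU emitter with two mesh renderers.
# Looping spawn rate follows an age curve, scaled randomly and by time-of-day user parameters.
# Birds spawn in a sphere, launch along a user velocity vector, and drift upward under a speed clamp.
# A visibility tag separates morning/day from evening/night, and mesh size grows with scalability distance to keep distant birds readable.
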NiagaraSystem /Game/Environments/Helios/FX/Olympus/Effects/NS_Env_Olympus_Doves_Flyby
dependencies {usedDataInterfaces [&NiagaraDataInterfaceCurve], usedDynamicInputs [], usedModules [&SystemState]}
emitters {Birds {
  simTarget=GPUComputeSim
  emitterUpdateScript [
   &EmitterState(
    Life Cycle Mode=::NewEnumerator1
    Loop Behavior=::NewEnumerator1
    Loop Delay=&RandomRangeFloat(Minimum=0.4, Maximum=0.6)
    Loop Duration Mode=::NewEnumerator1
    Reset Age On Awaken=true
    Scalability Mode=::NewEnumerator1
    UseLoopDelay=true
   )
   &SpawnRate(
    Spawn Probability=0.25
    SpawnRate=&FloatFromCurve(
     FloatCurve=DataInterface() {propertyValues {Curve {keys [
         <at 0 0.1 RCIM_Cubic>
         {arriveTangent=-0.0524, interpMode=RCIM_Cubic, leaveTangent=-0.0524, tangentMode=RCTM_User, time=1, value=0.1}
         {arriveTangent=5.5923, interpMode=RCIM_Cubic, leaveTangent=-0.0954, tangentMode=RCTM_Break, time=1.5, value=1}
         {arriveTangent=-0.1283, interpMode=RCIM_Cubic, leaveTangent=-1.8301, tangentMode=RCTM_Break, time=2, value=1}
         {interpMode=RCIM_Cubic, tangentMode=RCTM_Break, time=4, value=0}
        ]}}}
     CurveIndex=$System.Age
     Scale Curve=&Multiply_Float(
      A=&RandomRangeFloat(Minimum=150, Maximum=70)
      B=&FNTimeOfDayScalar(Night Scale=$User.NightScale, Morning Scale=$User.MorningScale, Day Scale=$User.DayScale, Evening Scale=$User.EveningScale)
     )
    )
    Use Spawn Probability=false
   )
  ]
  particleSpawnScript [
   /SetVariables(Particles.VisibilityTag=&Multiply_Int(
     A=1
     B=&MakeIntFromFloat(Float=&FN_Select_Float_from_Time_Of_Day(Morning=1, Day=1, Evening=0, Night=0))
    )) {isSetParameters=true}
   &InitializeParticle(
    Lifetime=&RandomRangeFloat(Minimum=10, Maximum=13)
    Mesh Scale Max=<1.1 1.1 1.1>
    Mesh Scale Mode=::NewEnumerator2
    Sprite Size Mode=::NewEnumerator3
    Uniform Sprite Size=500
    UsePositionOffset=false
    Write Material Random=true
   )
   &SphereLocation(
    MaskSpawnGroup=false
    Override Seed=false
    Sphere Radius=500
    Surface Only=false
    UseNonUniformScale=false
    UseOffset=false
    UseSphereOrientation=false
   )
   &AddVelocity(Velocity=$User.VelocityDirection)
   &DynamicMaterialParameters(
    Index 0 Param 1=0.25
    Index 0 Param 2=&Lerp_Float(
     A=&RandomRangeFloat(Minimum=2, Maximum=3)
     B=&RandomRangeFloat(Minimum=1, Maximum=1.5)
     Alpha=&MakeFloatFromInt(Int=$Particles.VisibilityTag)
    )
    Index 0 Param 3=0.25
    Index 0 Param 4=0.1
    Index 1 Param 1=1
    Index 1 Param 2=&RandomRangeFloat(Minimum=8, Maximum=10)
    Index 1 Param 3=2
    Index 2 Param 1=0.1
    Index 2 Param 2=1
    Index 2 Param 3=2
    Index 2 Param 4=0.25
    Index 3 Param 1=&RandomRangeFloat(Minimum=1.5, Maximum=2)
    Index 3 Param 4=1
    Module.Param0WriteEnabled=true
    Module.Param0WriteEnabled1=true
    Module.Param0WriteEnabled2=true
    Module.Param0WriteEnabled3=true
    Module.Param1WriteEnabled=true
    Module.Param1WriteEnabled1=true
    Module.Param1WriteEnabled2=true
    Module.Param1WriteEnabled3=true
    Module.Param2WriteEnabled=true
    Module.Param2WriteEnabled1=true
    Module.Param2WriteEnabled2=true
    Module.Param2WriteEnabled3=true
    Module.Param3WriteEnabled=true
    Module.Param3WriteEnabled1=true
    Module.Param3WriteEnabled2=true
    Module.Param3WriteEnabled3=true
    Write Parameter Index 1=true
    Write Parameter Index 2=true
    Write Parameter Index 3=true
   ) {enabled=false, module=DynamicMaterialParameters001}
   &DynamicMaterialParameters(
    Index 0 Param 2=1
    Index 0 Param 3=0.25
    Index 3 Param 1=1
    Index 3 Param 4=1
    Module.Param0WriteEnabled=true
    Module.Param0WriteEnabled1=true
    Module.Param0WriteEnabled2=true
    Module.Param0WriteEnabled3=true
    Module.Param1WriteEnabled=true
    Module.Param1WriteEnabled1=true
    Module.Param1WriteEnabled2=true
    Module.Param1WriteEnabled3=true
    Module.Param2WriteEnabled=true
    Module.Param2WriteEnabled1=true
    Module.Param2WriteEnabled2=true
    Module.Param2WriteEnabled3=true
    Module.Param3WriteEnabled=true
    Module.Param3WriteEnabled1=true
    Module.Param3WriteEnabled2=true
    Module.Param3WriteEnabled3=true
    Write Parameter Index 1=true
    Write Parameter Index 2=true
    Write Parameter Index 3=true
   )
  ]
  particleUpdateScript [
   &ParticleState()
   /SetVariables(Particles.RandomDirection:Vector3f=&Lerp_Vector(
     A=&MakeVector(
      X=&Waveform(
       Global Phase=&Add_Float(A=$Particles.Age, B=&RandomRangeFloat(Minimum=0, Maximum=5))
       "Waveform [1]"=::NewEnumerator8
       "[1] Frequency"=1
       "[1] Amplitude Min/Max"=<-1 1>
       "[2] Phase"=$Emitter.Age
       "[3] Phase"=$Emitter.Age
      )
      Y=&Waveform(
       Global Phase=&Add_Float(A=$Particles.Age, B=&RandomRangeFloat(Minimum=0, Maximum=5))
       "Waveform [1]"=::NewEnumerator8
       "[1] Frequency"=1
       "[1] Amplitude Min/Max"=<-1 1>
       "[2] Phase"=$Emitter.Age
       "[3] Phase"=$Emitter.Age
      )
      Z=0
     )
     B=<1 0 0>
     Alpha=&Waveform(
      Global Phase=&Add_Float(A=$Particles.Age, B=&RandomRangeFloat(Minimum=0, Maximum=5))
      Global Clamp Min=0.4
      "Waveform [1]"=::NewEnumerator8
      "[1] Frequency"=1
      "[1] Amplitude Min/Max"=<0 1>
      "[2] Phase"=$Emitter.Age
      "[3] Phase"=$Emitter.Age
     )
    )) {isSetParameters=true}
   &InterpolateOverTime(
    Target Vector Value:Vector3f=&NormalizeVector(Vector To Normalize=$Particles.RandomDirection, Safe Normalize=true, Safe Fallback Threshold=0, Safe Fallback Vector=<0 0 0.001>)
    Variable Type=::NewEnumerator2
   ) {module=InterpolateOverTime001}
   &WindForce(Air Resistance=0, Wind Speed=<0 0 10>)
   &SolveForcesAndVelocity(Clamp Velocity=true, Limit Acceleration=false, Speed Limit=800)
   &RegionMask(Outer Vector:Vector3f=<0 0 2>, Region Radius=2500, UseOffset=false, UseRotation=false, Value Type:ENiagaraRegionMaskValue=::NewEnumerator1)
   /SetVariables(Particles.New Vector 3f:Vector3f=&Multiply_VectorByFloat(
     Vector=$Particles.Scale
     Float=&FloatFromCurve(FloatCurve=DataInterface() {propertyValues {Curve {keys [
          {interpMode=RCIM_Cubic, tangentMode=RCTM_User, time=0, value=0}
          {interpMode=RCIM_Cubic, tangentMode=RCTM_User, time=0.1, value=1}
          {interpMode=RCIM_Cubic, tangentMode=RCTM_User, time=0.8, value=1}
          {tangentMode=RCTM_User, time=1, value=0}
         ]}}})
    )) {isSetParameters=true}
   &ScaleMeshSize(Scale Factor=&VectorFromFloat(Value=&FloatFromCurve(
      FloatCurve=DataInterface() {propertyValues {Curve {keys [
          {arriveTangent=3.3778, interpMode=RCIM_Cubic, leaveTangent=3.3778, tangentMode=RCTM_User, time=0, value=1}
          {arriveTangent=-0.2868, interpMode=RCIM_Cubic, leaveTangent=-0.2868, tangentMode=RCTM_User, time=1, value=2}
         ]}}}
      CurveIndex=$Emitter.Scalability.DistanceFraction
     )))
  ]
  rendererClasses [&NiagaraMeshRendererProperties]
  renderers [
   {rendererClass=&NiagaraMeshRendererProperties, rendererIndex=0}
   {rendererClass=&NiagaraMeshRendererProperties, rendererIndex=1}
  ]
 }}
summary {
 systemName=NS_Env_Olympus_Doves_Flyby
 emitters [{emitterName=Birds, simTarget=GPUComputeSim, rendererClasses [&NiagaraMeshRendererProperties]}]
 userVariables [
  User.DayScale {type=NiagaraFloat, defaultValue=1, description=""}
  User.EveningScale {type=NiagaraFloat, defaultValue=0.5, description=""}
  User.MorningScale {type=NiagaraFloat, defaultValue=1, description=""}
  User.NightScale {type=NiagaraFloat, defaultValue=0.5, description=""}
  User.VelocityDirection {type=Vector3f, description="", defaultValue {struct=Vector3f, value=<-2500 0 2000>}}
 ]
}
