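# Niagara system that releases GPU mesh doves over the Olympus environment as a camera-triggered ambient effect.
# A timed spawn curve is scaled by time of day, distance and the system's CameraTrigger value, and a mask texture filters spawn positions within a large box.
# Birds launch along a user velocity, then swirl under wind and an oscillating vortex, growing with camera distance.
# Two disabled debug emitters visualize CameraFade and the masked area.
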NiagaraSystem /Game/Environments/Helios/FX/Olympus/Effects/NS_Env_Olympus_Doves_LocationBased
dependencies {
 usedDataInterfaces [
  &NiagaraDataInterfaceColorCurve
  &NiagaraDataInterfaceGBuffer
  /Script/Niagara.NiagaraDataInterfaceTexture
  &NiagaraDataInterfaceCurve
  &NiagaraDataInterfaceCamera
 ]
 usedDynamicInputs [
  &RemapRange
  &DotProduct
  &SubtractVectorFromPosition
  &ReturnCameraPosition
 ]
 usedModules [
  &SystemState
  &self:SystemSpawnScript.SystemScriptSource.SystemScriptGraph.NiagaraNodeAssignment_1.SetVariables_#1
 ]
}
emitters {
 Birds001 {
  simTarget=GPUComputeSim
  emitterUpdateScript [
   &EmitterState(
    Enable Distance Culling=true
    Life Cycle Mode=::NewEnumerator1
    Loop Behavior=::NewEnumerator1
    Loop Delay=&RandomRangeFloat(Minimum=0.4, Maximum=0.6)
    Loop Duration Mode=::NewEnumerator1
    MaxDistance=29000
    MinDistance=15000
    Module.bApplySpawnCountScaleByDistance=true
    Module.bUseMaxDistance=true
    Module.bUseMinDistance=true
    Reset Age On Awaken=true
    Scalability Mode=::NewEnumerator1
    Scale Spawn Count=true
    Spawn Count Scale By Distance=DataInterface() {propertyValues {Curve {keys [<at 0.1 0>,<at 0.2 0.95>,<at 0.95 1>,<at 1 0>]}}}
    UseLoopDelay=true
   )
   &SpawnRate(
    SpawnRate=&FloatFromCurve(
     FloatCurve=DataInterface() {propertyValues {Curve {keys [
         {arriveTangent=9.0725, interpMode=RCIM_Cubic, leaveTangent=-0.055, tangentMode=RCTM_Break, time=0, value=0.01}
         {arriveTangent=0.0815, interpMode=RCIM_Cubic, leaveTangent=0.1702, tangentMode=RCTM_Break, time=0.5, value=0.03}
         {arriveTangent=2.6865, interpMode=RCIM_Cubic, leaveTangent=-0.5151, tangentMode=RCTM_Break, time=1.4, value=1}
         {arriveTangent=0.4412, interpMode=RCIM_Cubic, leaveTangent=0.4412, time=2.4, value=0.25}
         {arriveTangent=-0.119, interpMode=RCIM_Cubic, leaveTangent=-0.119, time=3.1, value=1.75}
         {arriveTangent=-0.0405, interpMode=RCIM_Cubic, leaveTangent=-0.0405, tangentMode=RCTM_User, time=4.5, value=0}
        ]}}}
     CurveIndex=$System.Age
     Scale Curve=&Multiply_Float(
      A=&Multiply_Float(
       A=&RandomRangeFloat(Minimum=5000, Maximum=3500)
       B=&FNTimeOfDayScalar(Night Scale=$User.NightScale, Morning Scale=$User.MorningScale, Day Scale=$User.DayScale, Evening Scale=$User.EveningScale)
      )
      B=$System.CameraTrigger
     )
    )
    Use Spawn Probability=false
   )
  ]
  particleSpawnScript [
   &InitializeParticle(
    Lifetime=&RandomRangeFloat(Minimum=10, Maximum=13)
    Mesh Scale Max=<1.1 1.1 1.1>
    Mesh Scale Mode=::NewEnumerator2
    Sprite Size Mode=::NewEnumerator3
    Uniform Sprite Size=500
    UsePositionOffset=false
    Write Material Random=true
   ) {module=InitializeParticle001}
   &ShapeLocation(
    Box Size=<25000 25000 100>
    MaskSpawnGroup=false
    Override Seed=false
    Rotation Coordinate Space=::World
    Shape Origin=&AddVectorToPosition(Position=$StackContext.Position, Vector=<0 0 0>)
    Shape Primitive=::NewEnumerator2
    Surface Only=false
    UseOwnerScale=true
   )
   &AddVelocity(Velocity=$User.VelocityDirection)
   &self:RainTextureMasking(
    AlphaThreshold=0.5
    BoxCenter:NiagaraPosition=<-57231.7812 66967.6562 12620.3506>
    BoxExtents:Vector3f=&Multiply_Vector(A=<19500.0781 19500.0781 1>, B=&VectorFromFloat(Value=1.0685))
    Position:NiagaraPosition=&AddVectorToPosition(Position=$Particles.Position, Vector=<-287.5273 -1516.1176 0>)
    RejectionTexture=DataInterface() {propertyValues {
      Texture=/Game/Environments/Helios/FX/Olympus/Textures/T_Olympus_Bird_Mask
      TextureUserParameter {parameter {name=None, typeDefHandle {registeredTypeIndex=2}, varData []}}
     }}
   ) {module=RainTextureMasking}
   &DynamicMaterialParameters(
    Index 0 Param 1=0.25
    Index 0 Param 2=&RandomRangeFloat(Minimum=1, Maximum=1.5)
    Index 0 Param 3=0.25
    Index 0 Param 4=0.1
    Index 1 Param 1=1
    Index 1 Param 2=&RandomRangeFloat(Minimum=8, Maximum=10)
    Index 1 Param 3=2
    Index 2 Param 1=0.1
    Index 2 Param 2=1
    Index 2 Param 3=2
    Index 2 Param 4=0.25
    Index 3 Param 1=&RandomRangeFloat(Minimum=1.5, Maximum=2)
    Index 3 Param 4=1
    Module.Param0WriteEnabled=true
    Module.Param0WriteEnabled1=true
    Module.Param0WriteEnabled2=true
    Module.Param0WriteEnabled3=true
    Module.Param1WriteEnabled=true
    Module.Param1WriteEnabled1=true
    Module.Param1WriteEnabled2=true
    Module.Param1WriteEnabled3=true
    Module.Param2WriteEnabled=true
    Module.Param2WriteEnabled1=true
    Module.Param2WriteEnabled2=true
    Module.Param2WriteEnabled3=true
    Module.Param3WriteEnabled=true
    Module.Param3WriteEnabled1=true
    Module.Param3WriteEnabled2=true
    Module.Param3WriteEnabled3=true
    Write Parameter Index 1=true
    Write Parameter Index 2=true
    Write Parameter Index 3=true
   ) {module=DynamicMaterialParameters001}
  ]
  particleUpdateScript [
   &ParticleState()
   /SetVariables(Particles.RandomDirection:Vector3f=&Lerp_Vector(
     A=&MakeVector(
      X=&Waveform(
       Global Phase=&Add_Float(A=$Particles.Age, B=&RandomRangeFloat(Minimum=0, Maximum=5))
       "Waveform [1]"=::NewEnumerator8
       "[1] Frequency"=1
       "[1] Amplitude Min/Max"=<-1 1>
       "[2] Phase"=$Emitter.Age
       "[3] Phase"=$Emitter.Age
      )
      Y=&Waveform(
       Global Phase=&Add_Float(A=$Particles.Age, B=&RandomRangeFloat(Minimum=0, Maximum=5))
       "Waveform [1]"=::NewEnumerator8
       "[1] Frequency"=1
       "[1] Amplitude Min/Max"=<-1 1>
       "[2] Phase"=$Emitter.Age
       "[3] Phase"=$Emitter.Age
      )
      Z=0
     )
     B=<1 0 0>
     Alpha=&Waveform(
      Global Phase=&Add_Float(A=$Particles.Age, B=&RandomRangeFloat(Minimum=0, Maximum=5))
      Global Clamp Min=0.4
      "Waveform [1]"=::NewEnumerator8
      "[1] Frequency"=1
      "[1] Amplitude Min/Max"=<0 1>
      "[2] Phase"=$Emitter.Age
      "[3] Phase"=$Emitter.Age
     )
    )) {isSetParameters=true}
   &InterpolateOverTime(
    Target Vector Value:Vector3f=&NormalizeVector(Vector To Normalize=$Particles.RandomDirection, Safe Normalize=true, Safe Fallback Threshold=0, Safe Fallback Vector=<0 0 0.001>)
    Variable Type=::NewEnumerator2
   ) {module=InterpolateOverTime001}
   &WindForce(Air Resistance=0, Wind Speed=<0 0 10>)
   &VortexForce(
    UseOriginOffset=false
    Vortex Axis=&Lerp_Vector(
     A=<222 222 3000>
     B=<-222 -222 -3000>
     Alpha=&Waveform("Waveform [1]"=::NewEnumerator0, "[1] Frequency"=$User.WaveformFrequency, "[1] Amplitude Min/Max"=<0 1>)
    )
   )
   &SolveForcesAndVelocity(Clamp Velocity=false, Limit Acceleration=false)
   /SetVariables(Particles.New Vector 3f:Vector3f=&Multiply_VectorByFloat(
     Vector=$Particles.Scale
     Float=&FloatFromCurve(
      FloatCurve=DataInterface() {propertyValues {Curve {keys [
          {interpMode=RCIM_Cubic, tangentMode=RCTM_User, time=0, value=0}
          {interpMode=RCIM_Cubic, tangentMode=RCTM_User, time=0.1, value=1}
          {interpMode=RCIM_Cubic, tangentMode=RCTM_User, time=0.8, value=1}
          {tangentMode=RCTM_User, time=1, value=0}
         ]}}}
      Scale Curve=&RandomRangeFloat(Minimum=0.6, Maximum=1)
     )
    )) {isSetParameters=true}
   &ScaleMeshSize(Scale Factor=&VectorFromFloat(Value=&FloatFromCurve(
      FloatCurve=DataInterface() {propertyValues {Curve {keys [
          {arriveTangent=3.3778, interpMode=RCIM_Cubic, leaveTangent=3.3778, tangentMode=RCTM_User, time=0, value=1}
          {arriveTangent=-0.2868, interpMode=RCIM_Cubic, leaveTangent=-0.2868, tangentMode=RCTM_User, time=1, value=2}
         ]}}}
      CurveIndex=$Emitter.Scalability.DistanceFraction
      Scale Curve=2
     )))
   &CameraOffset(Camera Offset Amount=-222)
  ]
  rendererClasses [&NiagaraMeshRendererProperties]
  renderers [{rendererClass=&NiagaraMeshRendererProperties, rendererIndex=0}]
 }
 DEBUG {
  bEnabled=false
  simTarget=GPUComputeSim
  emitterUpdateScript [
   &EmitterState(Life Cycle Mode=::NewEnumerator1, Loop Duration=1, UseLoopDelay=false)
   &SpawnRate(SpawnRate=&Multiply_Float(A=55, B=1), Use Spawn Probability=false)
   &SpawnBurst_Instantaneous(Use Spawn Probability=false, UseLoopCountLimit=false) {enabled=false}
  ]
  particleSpawnScript [
   &InitializeParticle(Color Mode=::NewEnumerator0, Sprite Size Mode=::NewEnumerator3, Uniform Sprite Size=200, UsePositionOffset=false, Write Material Random=true)
   &ShapeLocation(MaskSpawnGroup=false, Override Seed=false, Sphere Radius=900, UseOwnerScale=true)
  ]
  particleUpdateScript [
   &ParticleState()
   &SolveForcesAndVelocity(Clamp Velocity=false, Limit Acceleration=false)
   &ScaleColor(Scale Mode=::NewEnumerator1, ScaleRGBA=true) {enabled=false}
   &Color(Color=&LerpLinearColors(StartColor=<rgba 0 1 0.0988 1>, EndColor=<rgba 22 0.088 0.088 1>, LerpFactor=$System.CameraFade))
  ]
  rendererClasses [&NiagaraSpriteRendererProperties]
  renderers [{rendererClass=&NiagaraSpriteRendererProperties}]
 }
 DEBUG002 {
  bEnabled=false
  simTarget=GPUComputeSim
  emitterUpdateScript [
   &EmitterState(Life Cycle Mode=::NewEnumerator1, Loop Behavior=::NewEnumerator1, Loop Duration Mode=::NewEnumerator1, UseLoopDelay=false)
   &SpawnRate(SpawnRate=&Multiply_Float(A=50000, B=$System.CameraFade), Use Spawn Probability=false)
  ]
  particleSpawnScript [
   &InitializeParticle(Color Mode=::NewEnumerator0, Sprite Size Mode=::NewEnumerator3, Uniform Sprite Size=200, UsePositionOffset=false, Write Material Random=true)
   &ShapeLocation(
    Box Size=<25000 25000 100>
    MaskSpawnGroup=false
    Override Seed=false
    Rotation Coordinate Space=::World
    Shape Origin=&AddVectorToPosition(Position=$StackContext.Position, Vector=<0 0 0>)
    Shape Primitive=::NewEnumerator2
    Surface Only=false
    UseOwnerScale=true
   )
   /Game/Effects/Niagara/Modules/Spawn/FN_Spawn_GBuffer_Color_WithinRangeV2(HueMax=15, HueMin=9) {enabled=false}
   &self:RainTextureMasking(
    AlphaThreshold=0.5
    BoxCenter:NiagaraPosition=<-57231.7812 66967.6562 12620.3506>
    BoxExtents:Vector3f=&Multiply_Vector(A=<19500.0781 19500.0781 1>, B=&VectorFromFloat(Value=1.0685))
    Position:NiagaraPosition=&AddVectorToPosition(Position=$Particles.Position, Vector=<-287.5273 -1516.1176 0>)
    RejectionTexture=DataInterface() {propertyValues {
      Texture=&T_Olympus_Bird_Mask
      TextureUserParameter {parameter {name=None, typeDefHandle {registeredTypeIndex=2}, varData []}}
     }}
   ) {module=RainTextureMasking}
  ]
  particleUpdateScript [
   &ParticleState()
   &SolveForcesAndVelocity(Clamp Velocity=false, Limit Acceleration=false)
   &ScaleColor(Scale Mode=::NewEnumerator1, ScaleRGBA=true) {enabled=false}
   &Color(Color=&LerpLinearColors(StartColor=<rgba 1 1 1 1>, EndColor=<rgba 22 0.088 0.088 1>, LerpFactor=$System.CameraFade)) {module=Color001}
  ]
  rendererClasses [&NiagaraSpriteRendererProperties]
  renderers [{rendererClass=&NiagaraSpriteRendererProperties}]
 }
}
summary {
 systemName=NS_Env_Olympus_Doves_LocationBased
 emitters [
  {bEnabled=false, emitterName=DEBUG, simTarget=GPUComputeSim, rendererClasses [&NiagaraSpriteRendererProperties]}
  {bEnabled=false, emitterName=DEBUG002, simTarget=GPUComputeSim, rendererClasses [&NiagaraSpriteRendererProperties]}
  {emitterName=Birds001, simTarget=GPUComputeSim, rendererClasses [&NiagaraMeshRendererProperties]}
 ]
 userVariables [
  User.DayScale {type=NiagaraFloat, defaultValue=0.8, description=""}
  User.EveningScale {type=NiagaraFloat, defaultValue=0.1, description=""}
  User.MorningScale {type=NiagaraFloat, defaultValue=1, description=""}
  User.NightScale {type=NiagaraFloat, defaultValue=0.1, description=""}
  User.TriggerLocation {type=Vector3f, description="", defaultValue {struct=Vector3f, value=<0 0 0>}}
  User.VelocityDirection {type=Vector3f, description="", defaultValue {struct=Vector3f, value=<-72 1000 1000>}}
  User.WaveformFrequency {type=NiagaraFloat, defaultValue=0.2, description=""}
 ]
}
