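# Additive unlit Niagara sprite material that draws a procedural soft radial glow, shaped through dynamic parameters.
# An outer sphere mask with exponent and glow amount gets an optional center mask added, whose hardness input reuses its radius parameter.
# Pixel depth can scale down outer hardness, blurring the glow near the camera.
# The saturated mask multiplies particle color and alpha, with color passing through MF_FX_EmissiveAdjustments.
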Material /Game/Environments/Helios/FX/Olympus/Materials/M_SoftGlow_SmoothStep_NiagaraControlled_Additive
graph {
 properties {bUsedWithNiagaraSprites=true, blendMode=BLEND_Additive, pixelDepthOffsetMode=PDOM_Legacy, shadingModel=MSM_Unlit, usedShadingModels=Unlit}
 nodes [
  n1 = TextureCoordinate()
  n2 = Constant() {r=0.5}
  n3 = DynamicParameter() {defaultValue=<rgba 0.5 0.5 2 0.1>, paramNames [SphereMask_Radius, SphereMask_Hardness, SphereMask_Exponent, FakeGlowAmount]}
  n4 = DynamicParameter() {parameterIndex=2, defaultValue=<rgba 4096 0.25 1 1>, paramNames [BlurCamDist, UseBlurCamDistAmount, Param3, Param4]}
  n5 = Divide(A=PixelDepth().R, B=@n4.BlurCamDist) {constB=4096}
  n6 = LinearInterpolate(A=@n3.SphereMask_Hardness, B=Multiply(A=Saturate(None=@n5), B=@n3.SphereMask_Hardness), Alpha=@n4.UseBlurCamDistAmount)
  n7 = DynamicParameter() {defaultValue=<rgba 0.5 0.5 2 0.1>, paramNames [SphereMask_Radius, SphereMask_Hardness, SphereMask_Exponent, FakeGlowAmount]}
  n8 = Multiply(A=Power(Base=SphereMask(A=@n1, B=@n2, Radius=@n3.SphereMask_Radius, Hardness=@n6), Exp=@n3.SphereMask_Exponent), B=@n7.FakeGlowAmount) {constB=0.1}
  n9 = DynamicParameter() {
   parameterIndex=1
   defaultValue=<rgba 0.25 0.25 1 1>
   paramNames [SphereMaskCenter_Radius, SphereMaskCenter_Hardness, SphereMaskCenter_Exponent, SphereMaskCenter_Amount]
  }
  n10 = SphereMask(A=TextureCoordinate(), B=Constant() {r=0.5}, Radius=@n9.SphereMaskCenter_Radius, Hardness=@n9.SphereMaskCenter_Radius)
  n11 = DynamicParameter() {
   parameterIndex=1
   defaultValue=<rgba 0.25 0.25 1 1>
   paramNames [SphereMaskCenter_Radius, SphereMaskCenter_Hardness, SphereMaskCenter_Exponent, SphereMaskCenter_Amount]
  }
  n12 = DynamicParameter() {
   parameterIndex=1
   defaultValue=<rgba 0.25 0.25 1 1>
   paramNames [SphereMaskCenter_Radius, SphereMaskCenter_Hardness, SphereMaskCenter_Exponent, SphereMaskCenter_Amount]
  }
  n13 = LinearInterpolate(A=@n8, B=Add(A=@n8, B=Power(Base=@n10, Exp=@n11.SphereMaskCenter_Exponent)), Alpha=@n12.SphereMaskCenter_Amount)
  n14 = Multiply(A=@n13) {constB=0.5}
  ConstantBiasScale(None=Sine(None=Add(A=@n14) {constB=-0.25}))
  Constant() {r=1}
  n15 = ParticleColor()
  n16 = Saturate(None=@n13)
  MaterialFunctionCall(Emissive=Multiply(A=@n15.RGB, B=@n16)) {materialFunction=&MF_FX_EmissiveAdjustments}
  Multiply(A=@n15.A, B=@n16)
  Multiply()
  SphereMask()
 ]
}
